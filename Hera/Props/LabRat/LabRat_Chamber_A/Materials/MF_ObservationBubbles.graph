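# Material function for the LabRat observation chamber that renders animated bubbles in water, outputting Bubble Gradient, Bubble Highlight and Normal.
# Three panned, bump-offset bubble mask layers add fake parallax depth, with speeds blendable toward a "bubbly mode".
# Highlights combine sun-driven back and front specular with a Fresnel-faded bubble mask.
# The graph reuses Slurp parameters and textures, including pre-skinned triplanar parallax projection, and leaves several branches, including a water color composite, unconnected to outputs.
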MaterialFunction /Hera/Props/LabRat/LabRat_Chamber_A/Materials/MF_ObservationBubbles
graph {nodes [
  n1 = VectorParameter() {parameterName=Color_Water, defaultValue=<rgba 0 0.1295 0.4948 1>}
  n2 = VectorParameter() {parameterName=Color_Water_B, defaultValue=<rgba 0.0292 0.7712 0.8177 1>}
  n3 = MaterialFunctionCall(Power=ScalarParameter() {defaultValue=3, parameterName=WaterBG_FresnelPow}) {materialFunction=&Fresnel_Function}
  n4 = Power(Base=OneMinus(None=@n3.Result)) {constExponent=5}
  n5 = NamedRerouteUsage()
  n6 = VectorParameter() {parameterName=Color_Bubbles, defaultValue=<rgba 0 0.1865 0.5885 1>}
  n7 = VectorParameter() {parameterName=Color_Bubbles_B, defaultValue=<rgba 0.0208 0.5727 1.5625 1>}
  n8 = NamedRerouteUsage()
  n9 = CollectionParameter() {
   collection=&FortniteMaterialParameters
   parameterName=SunlightDirectionalVector
  }
  n10 = ComponentMask(None=@n9) {b=true, g=true, r=true}
  n11 = NamedRerouteUsage()
  n12 = Multiply(A=Multiply(A=DotProduct(A=@n10, B=@n11)) {constB=2}, B=@n11)
  n13 = DotProduct(
   A=Normalize(VectorInput=Subtract(A=@n12, B=@n10))
   B=Multiply(A=MaterialFunctionCall() {materialFunction=&CameraDirectionVector}.Result)
  )
  n14 = Power(Base=Max(A=@n13) {constB=0}, Exp=ScalarParameter() {defaultValue=5, group=Slurp, parameterName=Bubble_BackSpecular_Power})
  n15 = ScalarParameter() {defaultValue=0.2, group=Slurp, parameterName=Bubble_BackSpecular_Scale}
  n16 = Multiply(
   A=VectorParameter() {group=Slurp, parameterName=Bubble_BackSpecular_Color, defaultValue=<rgba 0.5 1 1 1>}.RGB
   B=Saturate(None=Multiply(A=@n14, B=@n15) {constB=0.15})
  )
  n17 = Power(Base=Multiply(A=@n13) {constB=-1}, Exp=ScalarParameter() {defaultValue=5, group=Slurp, parameterName=Bubble_HighlightSpecular_Power})
  n18 = ScalarParameter() {defaultValue=0.2, group=Slurp, parameterName=Bubble_HighlightSpecular_Scale}
  n19 = Multiply(
   A=Saturate(None=Multiply(A=@n17, B=@n18) {constB=0.55})
   B=VectorParameter() {group=Slurp, parameterName=Bubble_HighlightSpecular_Color, defaultValue=<rgba 0.5264 0.7332 1 1>}.RGB
  )
  n20 = MaterialFunctionCall() {materialFunction=&Fresnel_Function}
  n21 = Multiply(A=NamedRerouteUsage(), B=ScalarParameter() {defaultValue=-1, parameterName=Bubble_DistortionAmount1})
  n22 = NamedRerouteUsage()
  n23 = AppendVector(A=ScalarParameter() {defaultValue=3, parameterName=Bubble_Tiling_X1}, B=ScalarParameter() {defaultValue=2.5, parameterName=Bubble_Tiling_Y1})
  n24 = BumpOffset(
   Coordinate=Panner(Coordinate=Multiply(A=@n22, B=@n23), Speed=NamedRerouteUsage())
   Height=ScalarParameter() {defaultValue=0.1, parameterName=Bubble_BumpOffsetHeight_1}
   HeightRatioInput=ScalarParameter() {defaultValue=0.05, parameterName=Bubble_BumpOffsetHRI_1}
  )
  n25 = TextureObjectParameter() {
   group=Slurp
   parameterName=BubbleMask
   samplerType=SAMPLERTYPE_Masks
   texture=/Game/Characters/Player/Male/Large/Bodies/M_LRG_Slurp_Monster/Textures/T_SlurpBubble
  }
  n26 = TextureSample(UVs=NamedRerouteDeclaration(None=Add(A=@n21, B=@n24)) {name=Bubble_UV_1, nodeColor=<rgba 1 0 0.7266 1>}, Tex=@n25)
  n27 = Saturate(None=NamedRerouteDeclaration(None=MaterialFunctionCall(Float3=@n26.RGB) {materialFunction=&BreakOutFloat3Components}.R) {name=Bubble_Texture_1, nodeColor=<rgba 0.2188 1 0 1>})
  n28 = Multiply(
   A=NamedRerouteUsage()
   B=AppendVector(A=ScalarParameter() {defaultValue=1, parameterName=Bubble_Tiling_X2}, B=ScalarParameter() {defaultValue=0.8, parameterName=Bubble_Tiling_Y2})
  )
  n29 = NamedRerouteUsage()
  n30 = Add(
   A=Multiply(A=NamedRerouteUsage(), B=ScalarParameter() {defaultValue=0.2, parameterName=Bubble_DistortionAmount2})
   B=BumpOffset(
    Coordinate=Panner(Coordinate=@n28, Speed=@n29)
    Height=ScalarParameter() {defaultValue=-0.1, parameterName=Bubble_BumpOffsetHeight_2}
    HeightRatioInput=ScalarParameter() {defaultValue=0.05, parameterName=Bubble_BumpOffsetHRI_2}
   )
  )
  n31 = TextureObjectParameter() {
   group=Slurp
   parameterName=BubbleMask
   samplerType=SAMPLERTYPE_Masks
   texture=&T_SlurpBubble
  }
  n32 = MaterialFunctionCall(Float3=TextureSample(UVs=NamedRerouteDeclaration(None=@n30) {name=Bubble_UV_2, nodeColor=<rgba 1 0 0.7266 1>}, Tex=@n31).RGB) {materialFunction=&BreakOutFloat3Components}
  n33 = Saturate(None=Multiply(A=OneMinus(None=@n20.Result), B=Add(A=@n27, B=@n32.G)))
  n34 = NamedRerouteDeclaration(None=Multiply(A=Add(A=@n16, B=@n19), B=@n33)) {name=Bubble_Highlight, nodeColor=<rgba 1 0.1875 0 1>}
  Add(
   A=LinearInterpolate(
    A=LinearInterpolate(A=@n1.RGB, B=@n2.RGB, Alpha=@n3.Result)
    B=LinearInterpolate(A=@n6.RGB, B=@n7.RGB, Alpha=@n8)
    Alpha=Multiply(A=@n4, B=@n5)
   )
   B=@n34
  )
  FunctionOutput(None=@n8) {outputName=Bubble Gradient}
  FunctionOutput(None=@n34) {outputName=Bubble Highlight}
  n35 = Constant3Vector()
  n36 = TextureObjectParameter() {
   group=Slurp
   parameterName=BubbleNormal
   samplerType=SAMPLERTYPE_Normal
   texture=/Game/Characters/Player/Male/Large/Bodies/M_LRG_Slurp_Monster/Textures/T_SlurpBubble_N
  }
  n37 = TextureSample(UVs=NamedRerouteUsage(), Tex=@n36) {samplerType=SAMPLERTYPE_Normal}
  n38 = MaterialFunctionCall(X=@n37.R, Z=@n37.R) {materialFunction=&MakeFloat3}
  n39 = NamedRerouteUsage()
  n40 = ScalarParameter() {defaultValue=0.5, group=Slurp, parameterName=Bubble_Blend_NormalSharpness, sliderMax=1}
  n41 = Max(A=Add(A=Abs(None=@n39), B=Multiply(A=@n40) {constB=-0.55}) {constB=-0.55}) {constB=0}
  n42 = Divide(A=@n41, B=DotProduct(A=@n41, B=Constant3Vector() {constant=<rgba 1 1 1 0>}))
  n43 = ComponentMask(None=@n42) {g=true}
  n44 = TextureSample(UVs=NamedRerouteUsage(), Tex=@n36) {samplerType=SAMPLERTYPE_Normal}
  n45 = MaterialFunctionCall(Y=@n44.R, Z=@n44.R) {materialFunction=&MakeFloat3}
  n46 = ComponentMask(None=@n42) {r=true}
  n47 = TextureSample(UVs=NamedRerouteUsage(), Tex=@n36) {samplerType=SAMPLERTYPE_Normal}
  n48 = MaterialFunctionCall(X=@n47.R, Z=@n47.R) {materialFunction=&MakeFloat3}
  n49 = ComponentMask(None=@n42) {g=true}
  n50 = TextureSample(UVs=NamedRerouteUsage(), Tex=@n36) {samplerType=SAMPLERTYPE_Normal}
  n51 = MaterialFunctionCall(Y=@n50.R, Z=@n50.R) {materialFunction=&MakeFloat3}
  n52 = ComponentMask(None=@n42) {r=true}
  n53 = LinearInterpolate(
   A=Add(A=Multiply(A=@n38.Result, B=@n43), B=Multiply(A=@n45.Result, B=@n46))
   B=Add(A=Multiply(A=@n48.Result, B=@n49), B=Multiply(A=@n51.Result, B=@n52))
   Alpha=Saturate(None=NamedRerouteUsage())
  )
  n54 = NamedRerouteUsage()
  n55 = Normalize(VectorInput=Add(A=NamedRerouteUsage(), B=LinearInterpolate(A=@n35, B=@n53, Alpha=@n54)))
  FunctionOutput(None=NamedRerouteDeclaration(None=Transform(None=@n55) {transformSourceType=TRANSFORMSOURCE_Local}) {name=BubbleWSNormal, nodeColor=<rgba 1 0.9844 0 1>}) {outputName=Normal}
  n56 = TextureObjectParameter() {
   group=Slurp
   parameterName=BubbleMask
   samplerType=SAMPLERTYPE_Masks
   texture=&T_SlurpBubble
  }
  n57 = MaterialFunctionCall!(
   TextureObject=@n56
   TextureSize=NamedRerouteUsage()
   WorldPosition=Add(A=NamedRerouteUsage(), B=NamedRerouteUsage())
   Export Float 4
   In=NamedRerouteUsage()
   World Space Normal=NamedRerouteUsage()
   ProjectionTransitionContrast=ScalarParameter() {defaultValue=2, group=Slurp, parameterName=Bubble_Blend_SharpnessMasks, sliderMax=1}
  ) {materialFunction=/Game/Athena/SafeZone/Preskinned_WP_Texture_ScalePan}
  n58 = StaticSwitch(True=@n57.Z Texture, False=@n57.Z Texture, Value=StaticBoolParameter() {group=Slurp, parameterName=Bubble_FrontPlanarProjectionOnly})
  LinearInterpolate(
   A=MaterialFunctionCall(Float3=@n58) {materialFunction=&BreakOutFloat3Components}.R
   B=MaterialFunctionCall(f3=@n58) {materialFunction=&AddComponents}.f3
   Alpha=NamedRerouteUsage()
  )
  n59 = TextureObjectParameter() {
   group=Slurp
   parameterName=BubbleMask
   samplerType=SAMPLERTYPE_Masks
   texture=&T_SlurpBubble
  }
  n60 = MaterialFunctionCall(
   TextureObject=@n59
   TextureSize=NamedRerouteUsage()
   WorldPosition=Add(A=NamedRerouteUsage(), B=NamedRerouteUsage())
   In=NamedRerouteUsage()
   World Space Normal=NamedRerouteUsage()
   ProjectionTransitionContrast=ScalarParameter() {defaultValue=2, group=Slurp, parameterName=Bubble_Blend_SharpnessMasks, sliderMax=1}
  ) {materialFunction=&Preskinned_WP_Texture_ScalePan}
  n61 = StaticSwitch(True=@n60.Z Texture, False=@n60.Z Texture, Value=StaticBoolParameter() {group=Slurp, parameterName=Bubble_FrontPlanarProjectionOnly})
  LinearInterpolate(
   A=MaterialFunctionCall(Float3=@n61) {materialFunction=&BreakOutFloat3Components}.G
   B=MaterialFunctionCall(f3=@n61) {materialFunction=&AddComponents}.f3
   Alpha=NamedRerouteUsage()
  )
  n62 = TextureObjectParameter() {
   group=Slurp
   parameterName=BubbleMask
   samplerType=SAMPLERTYPE_Masks
   texture=&T_SlurpBubble
  }
  n63 = MaterialFunctionCall(
   TextureObject=@n62
   TextureSize=NamedRerouteUsage()
   WorldPosition=NamedRerouteUsage()
   In=NamedRerouteUsage()
   World Space Normal=NamedRerouteUsage()
   ProjectionTransitionContrast=ScalarParameter() {defaultValue=2, group=Slurp, parameterName=Bubble_Blend_SharpnessMasks, sliderMax=1}
  ) {materialFunction=&Preskinned_WP_Texture_ScalePan}
  n64 = StaticSwitch(True=@n63.Z Texture, False=@n63.Z Texture, Value=StaticBoolParameter() {group=Slurp, parameterName=Bubble_FrontPlanarProjectionOnly})
  LinearInterpolate(
   A=MaterialFunctionCall(Float3=@n64) {materialFunction=&BreakOutFloat3Components}.B
   B=MaterialFunctionCall(f3=@n64) {materialFunction=&AddComponents}.f3
   Alpha=NamedRerouteUsage()
  )
  Multiply(A=ComponentMask(None=TextureCoordinate()) {r=true})
  n65 = NamedRerouteUsage()
  n66 = ComponentMask(None=@n65) {g=true, r=true}
  n67 = Saturate(None=Sign(None=MaterialFunctionCall(XY=@n66) {materialFunction=&Swizzle}.YX))
  n68 = ScalarParameter() {group=Slurp, parameterName=TriPlanarSymmetryOffset}
  n69 = AppendVector(A=Add(A=ComponentMask(None=@n65) {g=true, r=true}, B=Multiply(A=@n67, B=@n68)), B=ComponentMask(None=@n65) {b=true})
  n70 = CameraPositionWS()
  n71 = Normalize(VectorInput=Subtract(A=@n65, B=TransformPosition(None=@n70) {transformSourceType=TRANSFORMPOSSOURCE_World}))
  n72 = Multiply(A=@n71, B=ScalarParameter() {defaultValue=15, group=Slurp, parameterName=Bubble_ParallaxDepth_Layer2}) {constB=15}
  n73 = Add(
   A=Add(A=@n69, B=Add(A=@n72) {constB=20})
   B=VectorParameter() {group=Slurp, parameterName=Bubble Position_2_Offset, defaultValue=<rgba 0 0 0 1>}.RGB
  )
  NamedRerouteDeclaration(None=@n73) {name=BubbleWorldPosParallax_2, nodeColor=<rgba 0.0781 0 1 1>}
  n74 = Add(A=Multiply(A=@n71, B=ScalarParameter() {defaultValue=15, group=Slurp, parameterName=Bubble_ParallaxDepth_Layer3}) {constB=15}) {constB=20}
  NamedRerouteDeclaration(None=Add(A=Add(A=@n69, B=@n74), B=VectorParameter() {group=Slurp, parameterName=Bubble Position_3_Offset, defaultValue=<rgba 0 0 0 1>}.RGB)) {name=BubbleWorldPosParallax_3, nodeColor=<rgba 0.0781 0 1 1>}
  n75 = ScalarParameter() {defaultValue=500, parameterName=NoiseTiling}
  n76 = MaterialFunctionCall!(
   WorldPosition=NamedRerouteUsage()
   World Space Normal
   UVSize_Front=@n75
   UVSize_Side=@n75
   UVSize_Top=@n75
   ProjectionTransitionContrast=Constant() {r=0.1}
  ) {materialFunction=/Game/Packages/Fortress_SharedMaterials/Base_Material_Functions/WorldAlignedTexture_UVsMasks}
  NamedRerouteDeclaration(None=@n76.Mask_Top) {name=Bubble_UVMask, nodeColor=<rgba 0.1016 0 1 1>}
  NamedRerouteDeclaration(None=FunctionInput(Preview=TextureCoordinate()) {bUsePreviewValueAsDefault=true, inputType=FunctionInput_Vector2}) {name=Coords, nodeColor=<rgba 0.5469 0 1 1>}
  n77 = Multiply(A=@n27, B=ScalarParameter() {defaultValue=1, group=Slurp, parameterName=BubbleIntensity_A}) {constB=0.5}
  n78 = Multiply(A=NamedRerouteUsage(), B=ScalarParameter() {defaultValue=0.25, parameterName=Bubble_DistortionAmount3})
  n79 = NamedRerouteUsage()
  n80 = AppendVector(A=ScalarParameter() {defaultValue=3, parameterName=Bubble_Tiling_X3}, B=ScalarParameter() {defaultValue=2.2, parameterName=Bubble_Tiling_Y3})
  n81 = BumpOffset(
   Coordinate=Panner(Coordinate=Multiply(A=@n79, B=@n80), Speed=NamedRerouteUsage())
   Height=ScalarParameter() {defaultValue=-0.5, parameterName=Bubble_BumpOffsetHeight_3}
   HeightRatioInput=ScalarParameter() {defaultValue=0.1, parameterName=Bubble_BumpOffsetHRI_3}
  )
  n82 = TextureObjectParameter() {
   group=Slurp
   parameterName=BubbleMask
   samplerType=SAMPLERTYPE_Masks
   texture=&T_SlurpBubble
  }
  n83 = TextureSample(UVs=NamedRerouteDeclaration(None=Add(A=@n78, B=@n81)) {name=Bubble_UV_3, nodeColor=<rgba 1 0 0.7266 1>}, Tex=@n82)
  NamedRerouteDeclaration(None=Add(
    A=Add(A=@n77, B=@n32.G)
    B=MaterialFunctionCall(Float3=@n83.G) {materialFunction=&BreakOutFloat3Components}.B
   )) {name=Msk_Bubbles, nodeColor=<rgba 1 0 0.6328 1>}
  n84 = PreSkinnedPosition()
  NamedRerouteDeclaration(None=ShaderStageSwitch(PixelShader=VertexInterpolator(VS=@n84).PS, VertexShader=@n84)) {name=PreSkinnedPos_VS-PS, nodeColor=<rgba 0 0.4844 1 1>}
  n85 = ScalarParameter() {defaultValue=12, group=Slurp, parameterName=Bubble_ParallaxDepth_Layer1}
  n86 = Add(
   A=Add(A=@n69, B=Multiply(A=@n71, B=@n85) {constB=12})
   B=VectorParameter() {group=Slurp, parameterName=Bubble Position_1_Offset, defaultValue=<rgba 0 0 0 1>}.RGB
  )
  NamedRerouteDeclaration(None=@n86) {name=BubbleWorldPosParallax_1, nodeColor=<rgba 0.0781 0 1 1>}
  n87 = VectorParameter() {group=Slurp, parameterName=Bubble_DistortAmplitude_1, defaultValue=<rgba 1 2 0.01 1>}
  n88 = VectorParameter() {group=Slurp_BubblyMode, parameterName=Bubble_DistortAmplitude_1_Bubbly, defaultValue=<rgba 0.5 0.5 0.02 1>}
  n89 = NamedRerouteUsage()
  n90 = LinearInterpolate(A=@n87.RGB, B=@n88.RGB, Alpha=@n89)
  n91 = StaticBoolParameter() {defaultValue=true, group=Slurp_BubblyMode, parameterName=EnableBubblyMode}
  n92 = ComponentMask(None=NamedRerouteUsage()) {b=true}
  n93 = ScalarParameter() {defaultValue=0.01, group=Slurp, parameterName=Bubble_DistortTiling}
  n94 = StaticSwitch(
   True=LinearInterpolate(A=@n93, B=ScalarParameter() {defaultValue=0.1, group=Slurp_BubblyMode, parameterName=Bubble_DistortTiling_Bubbly}, Alpha=@n89)
   False=@n93
   Value=@n91
  )
  n95 = ScalarParameter() {defaultValue=0.3, group=Slurp, parameterName=Bubble_DistortSpeed}
  n96 = ScalarParameter() {defaultValue=0.89, group=Slurp_BubblyMode, parameterName=Bubble_DistortSpeed_Bubbly}
  n97 = Multiply(A=Time(), B=StaticSwitch(True=LinearInterpolate(A=@n95, B=@n96, Alpha=@n89), False=@n95, Value=@n91)) {constB=0.5}
  n98 = Sine(None=Add(A=Multiply(A=@n92, B=@n94), B=Frac(None=@n97)))
  NamedRerouteDeclaration(None=Multiply(A=StaticSwitch(True=@n90, False=@n87.RGB, Value=@n91), B=@n98)) {name=BubbleDistort_1, nodeColor=<rgba 1 0 0.3281 1>}
  n99 = VectorParameter() {group=Slurp, parameterName=Bubble_DistortAmplitude_2, defaultValue=<rgba 2 0.9 1 1>}
  n100 = LinearInterpolate(
   A=@n99.RGB
   B=VectorParameter() {group=Slurp_BubblyMode, parameterName=Bubble_DistortAmplitude_2_Bubbly, defaultValue=<rgba 1 1 0.02 1>}.RGB
   Alpha=@n89
  )
  NamedRerouteDeclaration(None=Multiply(A=StaticSwitch(True=@n100, False=@n99.RGB, Value=@n91), B=@n98)) {name=BubbleDistort_2, nodeColor=<rgba 1 0 0.3281 1>}
  n101 = ScalarParameter() {defaultValue=3, parameterName=BubblesQuantSteps}
  n102 = Multiply(A=Saturate(None=ScalarParameter() {group=Slurp_BubblyMode, parameterName=BubblyMode_intensity}), B=@n101)
  NamedRerouteDeclaration(None=Divide(A=Floor(None=@n102), B=@n101)) {name=BubblyModeIntensityQuant, nodeColor=<rgba 0 1 0.5312 1>}
  NamedRerouteDeclaration(None=@n86) {name=BubbleWorldPosParallax_1_preSwizzle, nodeColor=<rgba 0.0781 0 1 1>}
  n103 = NamedRerouteUsage()
  n104 = MaterialFunctionCall(Float3=@n103) {materialFunction=&BreakOutFloat3Components}
  NamedRerouteDeclaration(None=StaticSwitchParameter(
    True=MaterialFunctionCall(X=@n104.R, Y=@n104.R, Z=@n104.R) {materialFunction=&MakeFloat3}.Result
    False=@n103
   ) {group=Slurp, parameterName=Bubble Top Down Mapping}) {name=PreskinnedNrm-Interp-SwizzledForTopDownMapping, nodeColor=<rgba 1 0 0.9609 1>}
  n105 = Multiply(A=@n76.UV_Front) {constB=0.4}
  n106 = Frac(None=Multiply(A=Time(), B=ScalarParameter() {defaultValue=0.1, group=Slurp, parameterName=Bubble_SHScaleNoiseSpeed}) {constB=0.05})
  n107 = AppendVector(A=ComponentMask(None=@n105) {r=true}, B=Add(A=ComponentMask(None=@n105) {g=true}, B=@n106))
  n108 = TextureObjectParameter() {
   group=Slurp
   parameterName=Bubble_SHScaleNoise
   samplerType=SAMPLERTYPE_Grayscale
   texture=&LowResBlurredNoise
  }
  n109 = Multiply(A=@n76.UV_Side) {constB=0.4}
  n110 = AppendVector(A=ComponentMask(None=@n109) {r=true}, B=Add(A=ComponentMask(None=@n109) {g=true}, B=@n106))
  n111 = Multiply(A=LinearInterpolate(
    A=TextureSample(UVs=@n107, Tex=@n108) {constMipValue=0, samplerType=SAMPLERTYPE_Grayscale, texture=&LowResBlurredNoise}.R
    B=TextureSample(UVs=@n110, Tex=@n108) {constMipValue=0, samplerType=SAMPLERTYPE_Grayscale, texture=&LowResBlurredNoise}.R
   ))
  NamedRerouteDeclaration(None=Power(Base=Saturate(None=@n111), Exp=ScalarParameter() {defaultValue=2, parameterName=NoisePower}) {constExponent=4}) {name=Bubble_Noise, nodeColor=<rgba 0 0.4609 1 1>}
  NamedRerouteDeclaration(None=@n73) {name=BubbleWorldPosParallax_2_preSwizzle, nodeColor=<rgba 0.0781 0 1 1>}
  n112 = Multiply(
   A=ScalarParameter() {group=Slurp_BubblyMode, parameterName=BubblyMode_intensity}
   B=ScalarParameter() {defaultValue=0.5, parameterName=UnmaskAllBubbles}
  )
  NamedRerouteDeclaration(None=StaticSwitch(
    True=Saturate(None=@n112)
    False=Constant()
    Value=StaticBoolParameter() {defaultValue=true, group=Slurp_BubblyMode, parameterName=EnableBubblyMode}
   )) {name=BubblyModeIntensity, nodeColor=<rgba 0.7812 1 0 1>}
  NamedRerouteDeclaration(None=@n33) {name=Bubble_PostFresnel, nodeColor=<rgba 0 1 0.3203 1>}
  n113 = MaterialFunctionCall!(ReflectionVector, ReflectionTexture) {materialFunction=/Engine/Functions/Engine_MaterialFunctions01/Reflections/ViewAlignedReflection}
  n114 = ScalarParameter() {defaultValue=1, group=Slurp, parameterName=Bubble_ViewDependantDistortion}
  n115 = Divide(A=Time(), B=ScalarParameter() {defaultValue=12, parameterName=Hide Speed})
  n116 = ScalarParameter() {defaultValue=0.025, group=Slurp, parameterName=Bubble_ViewDistortion_SpeedX}
  n117 = ScalarParameter() {defaultValue=0.005, group=Slurp, parameterName=Bubble_ViewDistortion_SpeedY}
  n118 = Rotator(Coordinate=Panner(Coordinate=Multiply(A=@n113.UVs, B=@n114), Time=@n115, Speed=AppendVector(A=@n116, B=@n117)), Time=@n115) {speed=0.0001}
  n119 = Multiply(
   A=Multiply(
    A=TextureSample(UVs=@n118) {constMipValue=0, samplerType=SAMPLERTYPE_Grayscale, texture=&LowResBlurredNoise}.G
    B=NamedRerouteUsage()
   )
   B=NamedRerouteUsage()
  )
  n120 = OneMinus(None=Fresnel(ExponentIn=ScalarParameter() {defaultValue=1, parameterName=Bubble_Fresnel_Power}) {baseReflectFraction=0, exponent=1})
  NamedRerouteDeclaration(None=Power(Base=Multiply(A=@n119, B=@n120)) {constExponent=0.6}) {name=Bubble_GradientMask, nodeColor=<rgba 0.6406 1 0 1>}
  n121 = ScalarParameter() {defaultValue=0.25, group=Slurp, parameterName=Bubble Speed X_1}
  n122 = ScalarParameter() {group=Slurp, parameterName=Bubble Speed Y_1}
  n123 = ScalarParameter() {defaultValue=2, group=Slurp, parameterName=BubbleSpeedGlobal}
  n124 = StaticSwitch(
   True=LinearInterpolate(A=@n123, B=ScalarParameter() {defaultValue=1, group=Slurp_BubblyMode, parameterName=BubbleSpeedGlobal_Bubbly}, Alpha=NamedRerouteUsage())
   False=@n123
   Value=StaticBoolParameter() {defaultValue=true, group=Slurp_BubblyMode, parameterName=EnableBubblyMode}
  )
  NamedRerouteDeclaration(None=Multiply(A=AppendVector(A=@n121, B=@n122), B=@n124)) {name=BubbleSpeed_1, nodeColor=<rgba 1 0 0.5625 1>}
  n125 = ScalarParameter() {defaultValue=0.25, group=Slurp, parameterName=Bubble Speed X_2}
  n126 = ScalarParameter() {group=Slurp, parameterName=Bubble Speed Y_2}
  NamedRerouteDeclaration(None=Multiply(A=AppendVector(A=@n125, B=@n126), B=@n124)) {name=BubbleSpeed_2, nodeColor=<rgba 1 0 0.5625 1>}
  n127 = ScalarParameter() {defaultValue=0.3, group=Slurp, parameterName=Bubble Speed X_3}
  n128 = ScalarParameter() {group=Slurp, parameterName=Bubble Speed Y_3}
  NamedRerouteDeclaration(None=Multiply(A=AppendVector(A=@n127, B=@n128), B=@n124)) {name=BubbleSpeed_3, nodeColor=<rgba 1 0 0.5625 1>}
  n129 = ScalarParameter() {defaultValue=2.5, group=Slurp, parameterName=BubblesSizeGlobal}
  n130 = StaticSwitch(
   True=LinearInterpolate(A=@n129, B=ScalarParameter() {defaultValue=0.33, group=Slurp_BubblyMode, parameterName=BubblesSizeGlobalBubbly}, Alpha=NamedRerouteUsage())
   False=@n129
   Value=StaticBoolParameter() {defaultValue=true, group=Slurp_BubblyMode, parameterName=EnableBubblyMode}
  )
  NamedRerouteDeclaration(None=Multiply(A=@n130, B=ScalarParameter() {defaultValue=35, group=Slurp, parameterName=BubbleSize_A})) {name=BubbleSize_A, nodeColor=<rgba 0.0312 0 1 1>}
  n131 = PreSkinnedNormal()
  NamedRerouteDeclaration(None=ShaderStageSwitch(PixelShader=VertexInterpolator(VS=@n131).PS, VertexShader=@n131)) {name=PreSkinnedNrm_VS-PS, nodeColor=<rgba 0 0.4844 1 1>}
  NamedRerouteDeclaration(None=Multiply(A=@n130, B=ScalarParameter() {defaultValue=20, group=Slurp, parameterName=BubbleSize_A_1})) {name=BubbleSize_A_1, nodeColor=<rgba 0.0312 0 1 1>}
  NamedRerouteDeclaration(None=Multiply(A=@n130, B=ScalarParameter() {defaultValue=15, group=Slurp, parameterName=BubbleSize_A_2})) {name=BubbleSize_A_2, nodeColor=<rgba 0.0312 0 1 1>}
  NamedRerouteUsage()
  n132 = NamedRerouteUsage()
  n133 = Divide(A=Add(A=NamedRerouteUsage(), B=NamedRerouteUsage()), B=Multiply(A=Abs(None=@n132)) {constB=-1})
  n134 = Time()
  n135 = NamedRerouteUsage()
  Panner(Coordinate=ComponentMask(None=@n133) {b=true, g=true}, Time=@n134, Speed=@n135)
  Panner(Coordinate=ComponentMask(None=@n133) {b=true, r=true}, Time=@n134, Speed=@n135)
  n136 = NamedRerouteUsage()
  n137 = Divide(A=Add(A=NamedRerouteUsage(), B=NamedRerouteUsage()), B=Multiply(A=Abs(None=@n136)) {constB=-1})
  n138 = NamedRerouteUsage()
  Panner(Coordinate=ComponentMask(None=@n137) {b=true, g=true}, Time=@n134, Speed=@n138)
  Panner(Coordinate=ComponentMask(None=@n137) {b=true, r=true}, Time=@n134, Speed=@n138)
  ScalarParameter() {defaultValue=1, parameterName=SSOpacity}
  TextureCoordinate()
  TextureCoordinate()
  VectorParameter() {parameterName=SubsurfaceColor, defaultValue=<rgba 0.0129 0.0267 0.1406 1>}
 ]}
